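# Material function for the WiryPerk pickaxe that layers an animated, glowing Tetris-style block effect over incoming material attributes.
# Time-driven flipbooks supply block color and mask, alongside a noise-modulated color-curve gradient, grid lines and UV-gradient sphere masks.
# Masked lerps blend emissive, base color, roughness, normal, metallic and specular, with a glass-sheet normal texture for surface detail.
# It borrows the WiryPerk glider's UV projection function and color atlas.
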MaterialFunction /BRCosmetics/Weapons/FORT_Melee/Pickaxe_WiryPerkUniversal/Materials/MF_Pickaxe_WiryPerk
graph {
 properties {description=InsideR}
 nodes [
  n1 = GetMaterialAttributes(None=FunctionInput() {inputName=Material, inputType=FunctionInput_MaterialAttributes}) {attributeGetTypes [
    #1
    #2
    #3
    #4
    #5
    #6
   ]}
  n2 = NamedRerouteUsage()
  n3 = Multiply(A=NamedRerouteUsage(), B=Saturate(None=NamedRerouteUsage()))
  n4 = SetMaterialAttributes!(
   MaterialAttributes=@n1.MaterialAttributes
   BaseColor=LinearInterpolate(A=@n1.BaseColor, Alpha=NamedRerouteUsage()) {constB=0}
   EmissiveColor=LinearInterpolate(A=@n1.EmissiveColor, B=Add(A=@n2, B=@n3), Alpha=NamedRerouteUsage())
   Roughness=LinearInterpolate(
    A=NamedRerouteUsage()
    B=ScalarParameter() {defaultValue=0.4, group=Pickaxe_WiryPerk, parameterName=Tetris_Roughness}
    Alpha=NamedRerouteUsage()
   )
   Normal=LinearInterpolate(A=NamedRerouteUsage(), B=NamedRerouteUsage(), Alpha=NamedRerouteUsage())
   Metallic=LinearInterpolate(A=NamedRerouteUsage(), Alpha=NamedRerouteUsage()) {constB=0.05}
   Specular=LinearInterpolate(A=NamedRerouteUsage(), Alpha=NamedRerouteUsage()) {constA=0.5, constB=0.5}
  ) {attributeSetTypes [
    #2
    #1
    #3
    #4
    #5
    #6
   ]}
  FunctionOutput(None=@n4)
  NamedRerouteDeclaration(None=@n1.Specular) {name=InputSpec, nodeColor=<rgba 0 1 0.625 1>}
  n5 = ComponentMask(None=TextureCoordinate()) {g=true}
  n6 = Constant() {r=1}
  n7 = ScalarParameter() {defaultValue=0.338, parameterName=bottomBlend}
  n8 = ScalarParameter() {defaultValue=0.4, parameterName=BottomHardness}
  n9 = ComponentMask(None=TextureCoordinate()) {g=true}
  n10 = Constant()
  n11 = ScalarParameter() {group=Pickaxe_WiryPerk, parameterName=TopMask}
  NamedRerouteDeclaration(None=Max(A=SphereMask(A=@n5, B=@n6, Radius=@n7, Hardness=@n8), B=SphereMask(A=@n9, B=@n10, Radius=@n11) {hardnessPercent=95})) {name=Animation Direction A, nodeColor=<rgba 0.7813 0 1 1>}
  n12 = Multiply(A=NamedRerouteUsage(), B=Constant() {r=16}) {constB=16}
  n13 = TextureSample(UVs=@n12) {texture=/BRCosmetics/Weapons/FORT_Melee/Pickaxe_WiryPerkUniversal/Textures/T_GridDebug_WiryPerk}
  NamedRerouteDeclaration(None=Multiply(A=@n13.R, B=VectorParameter() {group=Pickaxe_WiryPerk, parameterName=GridLineColor, defaultValue=<rgba 5 5 5 1>}.RGB)) {name=GridLine, nodeColor=<rgba 1 0.1875 0 1>}
  n14 = TextureCoordinate()
  n15 = Power(Base=NamedRerouteDeclaration(None=ComponentMask(None=@n14) {g=true}) {name=Mask_G, nodeColor=<rgba 0 1 0.1797 1>}, Exp=Constant() {r=0.53})
  n16 = LinearInterpolate(
   A=ScalarParameter() {defaultValue=-0.4557, group=Pickaxe_WiryPerk, parameterName=Gradient color MIn}
   B=LinearInterpolate(
    A=ScalarParameter() {defaultValue=1.1727, group=Pickaxe_WiryPerk, parameterName=Gradient color Max}
    B=ScalarParameter() {defaultValue=1.7, group=Pickaxe_WiryPerk, parameterName=Gradient color MaxB}
    Alpha=NamedRerouteUsage()
   )
   Alpha=@n15
  )
  n17 = Time()
  n18 = ScalarParameter() {defaultValue=0.2, parameterName=BaseEmissiveColorSpeed}
  n19 = TextureSample(UVs=NamedRerouteUsage()) {texture=/BRCosmetics/Weapons/FORT_Melee/Pickaxe_WiryPerkUniversal/Textures/T_GridNoise_WiryPerk}
  n20 = Sine(None=Add(A=Multiply(A=@n17, B=@n18) {constB=0.01}, B=@n19.R))
  n21 = Multiply(A=@n15, B=Saturate(None=Add(A=@n16, B=@n20)))
  n22 = NamedRerouteUsage()
  n23 = CurveAtlasRowParameter(CurveTime=ComponentMask(None=LinearInterpolate(A=@n21, Alpha=@n22) {constB=0}) {g=true}) {
   atlas=/BRCosmetics/Gadgets/Assets/VinderTech_GliderChute/Glider_WiryPerk/Materials/WiryPerk_ColorAtlas
   curve=/BRCosmetics/Gadgets/Assets/VinderTech_GliderChute/Glider_WiryPerk/Materials/WiryPerk_CurveBase
   group=06_Advanced Emissive
   parameterName=PanningEmissiveColorCurve
   sortPriority=6
  }
  NamedRerouteDeclaration(None=Multiply(A=@n23, B=ScalarParameter() {defaultValue=0.9, group=Pickaxe_WiryPerk, parameterName=BlockBrightness}) {constB=0.99}) {name=BaseEmissiveColor, nodeColor=<rgba 1 0 0.7969 1>}
  n24 = Time()
  n25 = NamedRerouteUsage()
  n26 = Constant() {r=8}
  n27 = TextureObject() {samplerType=SAMPLERTYPE_Masks, texture=/BRCosmetics/Weapons/FORT_Melee/Pickaxe_WiryPerkUniversal/Textures/T_Flipbook3_WiryPerk_alpha}
  n28 = MaterialFunctionCall(
   "Animation  Phase (0-1)"=Frac(None=Multiply(A=@n24, B=@n25) {constB=0.05})
   Number of Rows=@n26
   Number of Columns=@n26
   Texture=@n27
   UVs=NamedRerouteUsage()
  ) {materialFunction=&FlipBook}
  NamedRerouteDeclaration(None=OneMinus(None=MaterialFunctionCall(Float3=@n28.Result) {materialFunction=&BreakOutFloat3Components}.R)) {name=TetrisMask, nodeColor=<rgba 0 1 0.8828 1>}
  n29 = NamedRerouteUsage()
  n30 = NamedRerouteUsage()
  n31 = NamedRerouteUsage()
  n32 = Desaturation(None=NamedRerouteUsage(), Fraction=ScalarParameter() {group=Pickaxe_WiryPerk, parameterName=Desaturation})
  n33 = ScalarParameter() {defaultValue=2, group=Pickaxe_WiryPerk, parameterName=TetrisBrightness}
  n34 = VectorParameter() {group=Pickaxe_WiryPerk, parameterName=Bottom colorA, defaultValue=<rgba 0.0976 0.9 0.4853 1>}
  n35 = VectorParameter() {group=Pickaxe_WiryPerk, parameterName=Bottom colorB, defaultValue=<rgba 0.009 0.9 0.3065 1>}
  n36 = NamedRerouteUsage()
  n37 = SphereMask(A=NamedRerouteUsage(), B=Constant() {r=1}, Radius=Constant() {r=0.25}, Hardness=Constant() {r=0.15})
  n38 = LinearInterpolate(
   A=LinearInterpolate(A=Multiply(A=@n32, B=@n33) {constB=1.5}, B=@n29, Alpha=Multiply(A=@n30, B=@n31)) {constB=0}
   B=LinearInterpolate(A=@n29, B=LinearInterpolate(A=@n34.RGB, B=@n35.RGB, Alpha=@n36), Alpha=Saturate(None=@n37))
   Alpha=NamedRerouteUsage()
  )
  n39 = LinearInterpolate(A=Constant() {r=1}, Alpha=NamedRerouteUsage()) {constA=1, constB=0.8}
  NamedRerouteDeclaration(None=Multiply(A=Multiply(A=@n38, B=@n39), B=ScalarParameter() {defaultValue=0.9, group=Pickaxe_WiryPerk, parameterName=Global luminance})) {name=MainEmissive, nodeColor=<rgba 1 0.3984 0 1>}
  NamedRerouteDeclaration(None=@n19.R) {name=Pixel block, nodeColor=<rgba 1 0.3281 0 1>}
  n40 = MaterialFunctionCall!(
   Root_Position=Constant3Vector() {constant=<rgba 0 0 0.06 1>}
   UseClamp?=Constant()
   FlipUV=Constant()
   UVSize=ScalarParameter() {defaultValue=-31.7, group=Pickaxe_WiryPerk, parameterName=UVSize}
   CenterOffsetX=ScalarParameter() {defaultValue=-26.8918, group=Pickaxe_WiryPerk, parameterName=CenterOffsetx}
   CenterOffsetY=Constant()
   CenterOffsetZ=ScalarParameter() {defaultValue=-0.908, group=Pickaxe_WiryPerk, parameterName=CenterOffsetZ}
  ) {materialFunction=/BRCosmetics/Gadgets/Assets/VinderTech_GliderChute/Glider_WiryPerk/Materials/MF_ProjectUV_WiryPerk}
  NamedRerouteDeclaration(None=@n40.UV_BackProjection) {name=ProjectUVRotation, nodeColor=<rgba 0 1 0.8359 1>}
  n41 = Multiply(
   A=Time()
   B=NamedRerouteDeclaration(None=ScalarParameter() {defaultValue=0.08, group=Pickaxe_WiryPerk, parameterName=FlipbookSpeed}) {name=FlipbookSpeed, nodeColor=<rgba 0.3828 1 0 1>}
  ) {constB=0.05}
  n42 = Constant() {r=8}
  NamedRerouteDeclaration(None=MaterialFunctionCall(
    "Animation  Phase (0-1)"=Frac(None=@n41)
    Number of Rows=@n42
    Number of Columns=@n42
    Texture=TextureObject() {texture=/BRCosmetics/Weapons/FORT_Melee/Pickaxe_WiryPerkUniversal/Textures/T_Flipbook3_WiryPerk}
    UVs=NamedRerouteUsage()
   ) {materialFunction=&FlipBook}.Result) {name=Tetris, nodeColor=<rgba 0 0.9297 1 1>}
  NamedRerouteDeclaration(None=MaterialFunctionCall(
    Normal=TextureSample(UVs=@n12) {texture=/BRCosmetics/Weapons/FORT_Melee/Pickaxe_WiryPerkUniversal/Textures/T_GlassSheet_Normals}.RGB
    Flatness=ScalarParameter() {defaultValue=0.5, group=Pickaxe_WiryPerk, parameterName=NormalFlatness}
   ) {materialFunction=&FlattenNormal}.Result) {name=Normal, nodeColor=<rgba 1 0.9844 0 1>}
  NamedRerouteDeclaration(None=@n1.Normal) {name=InputNormal, nodeColor=<rgba 0.6641 0 1 1>}
  NamedRerouteDeclaration(None=@n1.Roughness) {name=InputRoughness, nodeColor=<rgba 1 0 0.3281 1>}
  n43 = TextureSample() {texture=/BRCosmetics/Weapons/FORT_Melee/Pickaxe_WiryPerkUniversal/Textures/T_WiryPerkUniversal_Pickaxe_TA_mask}
  NamedRerouteDeclaration(None=Max(A=@n43.R, B=@n43.R)) {name=SwordMask_R, nodeColor=<rgba 0 0.6484 1 1>}
  NamedRerouteDeclaration(None=@n43.G) {name=Blade tip side mask, nodeColor=<rgba 0.4297 1 0 1>}
  NamedRerouteDeclaration(None=@n40.UV_BackProjection) {name=ProjectUV, nodeColor=<rgba 1 0.0234 0 1>}
  NamedRerouteDeclaration(None=@n1.Metallic) {name=InputMetallic, nodeColor=<rgba 0.2422 1 0 1>}
  NamedRerouteDeclaration(None=LinearInterpolate(A=ScalarParameter() {defaultValue=0.75, parameterName=GlassMIN}, B=ScalarParameter() {defaultValue=1, parameterName=GalssMax}, Alpha=@n13.G) {constA=0.9}) {name=InsideR, nodeColor=<rgba 0 0.5781 1 1>}
  NamedRerouteDeclaration(None=ScalarParameter() {group=Pickaxe_WiryPerk, parameterName=Animation direction}) {name=Animation direction, nodeColor=<rgba 0 0.0625 1 1>}
 ]
}
